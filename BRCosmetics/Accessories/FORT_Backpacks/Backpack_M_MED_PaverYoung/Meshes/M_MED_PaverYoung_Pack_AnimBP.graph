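# Anim Blueprint for the PaverYoung backpack cosmetic, adding secondary physics motion to its dangling leg and head bones.
# The input pose passes through serial single-bone AnimDynamics nodes for twelve front/back, left/right leg bones, then a softer, partially blended head.
# Each simulates root-relative with angular springs, tight limits and reduced gravity.
# Update event sets stiffer leaves/head spring variables in the front end, but the nodes show fixed spring values, with no visible reads of those variables.
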AnimBlueprint /BRCosmetics/Accessories/FORT_Backpacks/Backpack_M_MED_PaverYoung/Meshes/M_MED_PaverYoung_Pack_AnimBP
target_skeleton=&/M_MED_PaverYoung_Pack_Skeleton
graphs [
 AnimGraph {nodes [
   n1 = Anim Dynamics - Bone: dyn_leg_front_01_l!(ComponentPose=Local To Component(LocalPose=Input Pose()), Alpha=1, AngularSpringConstant:float=10000, ExternalForce="0,0,0") -> (Pose) {
    angularDampingOverride=0.999
    angularSpringConstant=10000
    bAngularSpring=true
    bOverrideAngularDamping=true
    bUseGravityOverride=true
    gravityScale=0.35
    simulationSpace=RootRelative
    boundBone {boneName=dyn_leg_front_01_l}
    componentAppliedLinearAccClamp=<2000 2000 2000>
    componentLinearAccScale=<1 1 1>
    componentLinearVelScale=<1 1 1>
    gravityOverride {z=-980}
    physicsBodyDefinitions [{
      boundBone {boneName=dyn_leg_front_01_l}
      boxExtents=<20 30 30>
      constraintSetup {angularLimitsMax=<4 5 0>, angularLimitsMin=<-4 -12.5 -8>, angularTarget=<1 0 0>}
      localJointOffset=<2 0 0>
     }]
    simSpaceSettings {maxAngularAcceleration=100, maxAngularVelocity=50, simSpaceAngularAlpha=0.5}
   }
   n2 = Anim Dynamics - Bone: dyn_leg_front_02_l!(ComponentPose=@n1, Alpha=1, AngularSpringConstant:float=10000, ExternalForce="0,0,0") -> (Pose) {
    angularDampingOverride=0.999
    angularSpringConstant=10000
    bAngularSpring=true
    bOverrideAngularDamping=true
    bUseGravityOverride=true
    gravityScale=0.35
    simulationSpace=RootRelative
    boundBone {boneName=dyn_leg_front_02_l}
    componentAppliedLinearAccClamp=<2000 2000 2000>
    componentLinearAccScale=<1 1 1>
    componentLinearVelScale=<1 1 1>
    gravityOverride {z=-980}
    physicsBodyDefinitions [{
      boundBone {boneName=dyn_leg_front_02_l}
      boxExtents=<20 30 30>
      constraintSetup {angularLimitsMax=<4 5 0>, angularLimitsMin=<-4 -12.5 -8>, angularTarget=<1 0 0>}
      localJointOffset=<2 0 0>
     }]
    simSpaceSettings {maxAngularAcceleration=100, maxAngularVelocity=50, simSpaceAngularAlpha=0.5}
   }
   n3 = Anim Dynamics - Bone: dyn_leg_front_03_l!(ComponentPose=@n2, Alpha=1, AngularSpringConstant:float=10000, ExternalForce="0,0,0") -> (Pose) {
    angularDampingOverride=0.999
    angularSpringConstant=10000
    bAngularSpring=true
    bOverrideAngularDamping=true
    bUseGravityOverride=true
    gravityScale=0.35
    simulationSpace=RootRelative
    boundBone {boneName=dyn_leg_front_03_l}
    componentAppliedLinearAccClamp=<2000 2000 2000>
    componentLinearAccScale=<1 1 1>
    componentLinearVelScale=<1 1 1>
    gravityOverride {z=-980}
    physicsBodyDefinitions [{
      boundBone {boneName=dyn_leg_front_03_l}
      boxExtents=<20 30 30>
      constraintSetup {angularLimitsMax=<4 5 0>, angularLimitsMin=<-4 -12.5 -8>, angularTarget=<1 0 0>}
      localJointOffset=<2 0 0>
     }]
    simSpaceSettings {maxAngularAcceleration=100, maxAngularVelocity=50, simSpaceAngularAlpha=0.5}
   }
   n4 = Anim Dynamics - Bone: dyn_leg_back_01_l!(ComponentPose=@n3, Alpha=1, AngularSpringConstant:float=10000, ExternalForce="0,0,0") -> (Pose) {
    angularDampingOverride=0.9994
    angularSpringConstant=10000
    bAngularSpring=true
    bOverrideAngularDamping=true
    bUseGravityOverride=true
    gravityScale=0.35
    simulationSpace=RootRelative
    boundBone {boneName=dyn_leg_back_01_l}
    componentAppliedLinearAccClamp=<2000 2000 2000>
    componentLinearAccScale=<1 1 1>
    componentLinearVelScale=<1 1 1>
    gravityOverride {z=-980}
    physicsBodyDefinitions [{
      boundBone {boneName=dyn_leg_back_01_l}
      boxExtents=<20 30 30>
      constraintSetup {angularLimitsMax=<4 5 0>, angularLimitsMin=<-4 -5 -12>, angularTarget=<1 0 0>}
      localJointOffset=<2 0 0>
     }]
    simSpaceSettings {maxAngularAcceleration=100, maxAngularVelocity=50, simSpaceAngularAlpha=0.5}
   }
   n5 = Anim Dynamics - Bone: dyn_leg_back_02_l!(ComponentPose=@n4, Alpha=1, AngularSpringConstant:float=10000, ExternalForce="0,0,0") -> (Pose) {
    angularDampingOverride=0.9994
    angularSpringConstant=10000
    bAngularSpring=true
    bOverrideAngularDamping=true
    bUseGravityOverride=true
    gravityScale=0.35
    simulationSpace=RootRelative
    boundBone {boneName=dyn_leg_back_02_l}
    componentAppliedLinearAccClamp=<2000 2000 2000>
    componentLinearAccScale=<1 1 1>
    componentLinearVelScale=<1 1 1>
    gravityOverride {z=-980}
    physicsBodyDefinitions [{
      boundBone {boneName=dyn_leg_back_02_l}
      boxExtents=<20 30 30>
      constraintSetup {angularLimitsMax=<4 5 0>, angularLimitsMin=<-4 -5 -12>, angularTarget=<1 0 0>}
      localJointOffset=<-1 3 0>
     }]
    simSpaceSettings {maxAngularAcceleration=100, maxAngularVelocity=50, simSpaceAngularAlpha=0.5}
   }
   n6 = Anim Dynamics - Bone: dyn_leg_back_03_l!(ComponentPose=@n5, Alpha=1, AngularSpringConstant:float=10000, ExternalForce="0,0,0") -> (Pose) {
    angularDampingOverride=0.9994
    angularSpringConstant=10000
    bAngularSpring=true
    bOverrideAngularDamping=true
    bUseGravityOverride=true
    gravityScale=0.35
    simulationSpace=RootRelative
    boundBone {boneName=dyn_leg_back_03_l}
    componentAppliedLinearAccClamp=<2000 2000 2000>
    componentLinearAccScale=<1 1 1>
    componentLinearVelScale=<1 1 1>
    gravityOverride {z=-980}
    physicsBodyDefinitions [{
      boundBone {boneName=dyn_leg_back_03_l}
      boxExtents=<20 30 30>
      constraintSetup {angularLimitsMax=<4 5 0>, angularLimitsMin=<-4 -5 -12>, angularTarget=<1 0 0>}
      localJointOffset=<-1 3 0>
     }]
    simSpaceSettings {maxAngularAcceleration=100, maxAngularVelocity=50, simSpaceAngularAlpha=0.5}
   }
   n7 = Anim Dynamics - Bone: dyn_leg_front_01_r!(ComponentPose=@n6, Alpha=1, AngularSpringConstant:float=10000, ExternalForce="0,0,0") -> (Pose) {
    angularDampingOverride=0.9997
    angularSpringConstant=10000
    bAngularSpring=true
    bOverrideAngularDamping=true
    bUseGravityOverride=true
    gravityScale=0.35
    simulationSpace=RootRelative
    boundBone {boneName=dyn_leg_front_01_r}
    componentAppliedLinearAccClamp=<2000 2000 2000>
    componentLinearAccScale=<1 1 1>
    componentLinearVelScale=<1 1 1>
    gravityOverride {z=-980}
    physicsBodyDefinitions [{
      boundBone {boneName=dyn_leg_front_01_r}
      boxExtents=<20 30 30>
      constraintSetup {angularLimitsMax=<4 5 0>, angularLimitsMin=<-4 -12 -4>, angularTarget=<1 0 0>}
      localJointOffset=<-2 1.5 3>
     }]
    simSpaceSettings {maxAngularAcceleration=100, maxAngularVelocity=50, simSpaceAngularAlpha=0.5}
   }
   n8 = Anim Dynamics - Bone: dyn_leg_front_02_r!(ComponentPose=@n7, Alpha=1, AngularSpringConstant:float=10000, ExternalForce="0,0,0") -> (Pose) {
    angularDampingOverride=0.9997
    angularSpringConstant=10000
    bAngularSpring=true
    bOverrideAngularDamping=true
    bUseGravityOverride=true
    gravityScale=0.35
    simulationSpace=RootRelative
    boundBone {boneName=dyn_leg_front_02_r}
    componentAppliedLinearAccClamp=<2000 2000 2000>
    componentLinearAccScale=<1 1 1>
    componentLinearVelScale=<1 1 1>
    gravityOverride {z=-980}
    physicsBodyDefinitions [{
      boundBone {boneName=dyn_leg_front_02_r}
      boxExtents=<20 30 30>
      constraintSetup {angularLimitsMax=<4 5 0>, angularLimitsMin=<-4 -12 -4>, angularTarget=<1 0 0>}
      localJointOffset=<-2 0 0>
     }]
    simSpaceSettings {maxAngularAcceleration=100, maxAngularVelocity=50, simSpaceAngularAlpha=0.5}
   }
   n9 = Anim Dynamics - Bone: dyn_leg_front_03_r!(ComponentPose=@n8, Alpha=1, AngularSpringConstant:float=10000, ExternalForce="0,0,0") -> (Pose) {
    angularDampingOverride=0.9997
    angularSpringConstant=10000
    bAngularSpring=true
    bOverrideAngularDamping=true
    bUseGravityOverride=true
    gravityScale=0.35
    simulationSpace=RootRelative
    boundBone {boneName=dyn_leg_front_03_r}
    componentAppliedLinearAccClamp=<2000 2000 2000>
    componentLinearAccScale=<1 1 1>
    componentLinearVelScale=<1 1 1>
    gravityOverride {z=-980}
    physicsBodyDefinitions [{
      boundBone {boneName=dyn_leg_front_03_r}
      boxExtents=<20 30 30>
      constraintSetup {angularLimitsMax=<4 5 0>, angularLimitsMin=<-4 -12 -4>, angularTarget=<1 0 0>}
      localJointOffset=<-2 0 0>
     }]
    simSpaceSettings {maxAngularAcceleration=100, maxAngularVelocity=50, simSpaceAngularAlpha=0.5}
   }
   n10 = Anim Dynamics - Bone: dyn_leg_back_01_r!(ComponentPose=@n9, Alpha=1, AngularSpringConstant:float=10000, ExternalForce="0,0,0") -> (Pose) {
    angularDampingOverride=0.9994
    angularSpringConstant=10000
    bAngularSpring=true
    bOverrideAngularDamping=true
    bUseGravityOverride=true
    gravityScale=0.35
    simulationSpace=RootRelative
    boundBone {boneName=dyn_leg_back_01_r}
    componentAppliedLinearAccClamp=<2000 2000 2000>
    componentLinearAccScale=<1 1 1>
    componentLinearVelScale=<1 1 1>
    gravityOverride {z=-980}
    physicsBodyDefinitions [{
      boundBone {boneName=dyn_leg_back_01_r}
      boxExtents=<20 30 30>
      constraintSetup {angularLimitsMax=<4 5 0>, angularLimitsMin=<-4 -15 -8>, angularTarget=<1 0 0>}
      localJointOffset=<2 0 0>
     }]
    simSpaceSettings {maxAngularAcceleration=100, maxAngularVelocity=50, simSpaceAngularAlpha=0.5}
   }
   n11 = Anim Dynamics - Bone: dyn_leg_back_02_r!(ComponentPose=@n10, Alpha=1, AngularSpringConstant:float=10000, ExternalForce="0,0,0") -> (Pose) {
    angularDampingOverride=0.9994
    angularSpringConstant=10000
    bAngularSpring=true
    bOverrideAngularDamping=true
    bUseGravityOverride=true
    gravityScale=0.35
    simulationSpace=RootRelative
    boundBone {boneName=dyn_leg_back_02_r}
    componentAppliedLinearAccClamp=<2000 2000 2000>
    componentLinearAccScale=<1 1 1>
    componentLinearVelScale=<1 1 1>
    gravityOverride {z=-980}
    physicsBodyDefinitions [{
      boundBone {boneName=dyn_leg_back_02_r}
      boxExtents=<20 30 30>
      constraintSetup {angularLimitsMax=<4 5 0>, angularLimitsMin=<-4 -15 -8>, angularTarget=<1 0 0>}
      localJointOffset=<2 0 0>
     }]
    simSpaceSettings {maxAngularAcceleration=100, maxAngularVelocity=50, simSpaceAngularAlpha=0.5}
   }
   n12 = Anim Dynamics - Bone: dyn_leg_back_03_r!(ComponentPose=@n11, Alpha=1, AngularSpringConstant:float=10000, ExternalForce="0,0,0") -> (Pose) {
    angularDampingOverride=0.9994
    angularSpringConstant=10000
    bAngularSpring=true
    bOverrideAngularDamping=true
    bUseGravityOverride=true
    gravityScale=0.35
    simulationSpace=RootRelative
    boundBone {boneName=dyn_leg_back_03_r}
    componentAppliedLinearAccClamp=<2000 2000 2000>
    componentLinearAccScale=<1 1 1>
    componentLinearVelScale=<1 1 1>
    gravityOverride {z=-980}
    physicsBodyDefinitions [{
      boundBone {boneName=dyn_leg_back_03_r}
      boxExtents=<20 30 30>
      constraintSetup {angularLimitsMax=<4 5 0>, angularLimitsMin=<-4 -15 -8>, angularTarget=<1 0 0>}
      localJointOffset=<2 0 0>
     }]
    simSpaceSettings {maxAngularAcceleration=100, maxAngularVelocity=50, simSpaceAngularAlpha=0.5}
   }
   n13 = Anim Dynamics - Bone: dyn_head!(ComponentPose=@n12, Alpha=0.8, AngularSpringConstant:float=2000, ExternalForce="0,0,0") -> (Pose) {
    angularDampingOverride=0.999
    angularSpringConstant=2000
    bAngularSpring=true
    bOverrideAngularDamping=true
    bUseGravityOverride=true
    gravityScale=0.35
    simulationSpace=RootRelative
    boundBone {boneName=dyn_head}
    componentAppliedLinearAccClamp=<2000 2000 2000>
    componentLinearAccScale=<0.2 0.2 0.2>
    componentLinearVelScale=<1 1 1>
    gravityOverride {z=-980}
    physicsBodyDefinitions [{
      boundBone {boneName=dyn_head}
      boxExtents=<20 30 30>
      constraintSetup {angularLimitsMax=<10 0 8>, angularLimitsMin=<-10 0 -8>, angularTarget=<1 0 0>}
      localJointOffset=<4 0 0>
     }]
    simSpaceSettings {maxAngularAcceleration=100, maxAngularVelocity=50, simSpaceAngularAlpha=1}
   }
   Output Pose(Result=Component To Local(ComponentPose=@n13)) {layerGroup=None}
  ]}
 EventGraph {nodes [
   Event BlueprintUpdateAnimation()
   > n1 = Branch(Condition=Get bIsInFrontEnd())
   > Set LeavesAngularSpringConstant(LeavesAngularSpringConstant:double=13000) -> (Output_Get:double)
   > Set HeadAngularSpringConstant(HeadAngularSpringConstant:double=2500) -> (Output_Get:double)
   Set LeavesAngularSpringConstant(execute=@n1.else, LeavesAngularSpringConstant=10000)
   > Set HeadAngularSpringConstant(HeadAngularSpringConstant=2000)
  ]}
]
variables [
 HeadAngularSpringConstant {type=number}
 LeavesAngularSpringConstant {type=number}
]
